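# Physics asset for the F_MED_PolkaSkate female body, set up for skirt physics against the lower body and ground.
# Pelvis, thigh, calf and foot capsules provide body shapes, while a huge sphere offset downward approximates a flat ground plane.
# Eight skirt chains each link four capsules to a tiny MSTR anchor through swing and twist constraints.
# Front chains have free-rotation roots and cross-links at the second segment, tying left and right to the middle.
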PhysicsAsset /BRCosmetics/Characters/Player/Female/Medium/Bodies/F_MED_PolkaSkate/Meshes/F_MED_PolkaSkate_Physics
bodies {
 calf_l [{
   shapeType=Capsule
   length=41.1365
   radius=6.4413
   shapeName=calf_l_capsule
   center=<-27.3453 -2.6695 -0.1438>
   rotation=<rot -0.2431 91.5338 -90.0684>
  }]
 calf_r [{
   shapeType=Capsule
   length=41.1365
   radius=6.4413
   shapeName=calf_r_capsule
   center=<27.3453 2.6695 0.1438>
   rotation=<rot -0.2431 91.5338 89.9316>
  }]
 dyn_skirt_back_l_01 [{
   shapeType=Capsule
   length=42.0125
   radius=6.5688
   shapeName=dyn_skirt_back_l_01_capsule
   center=<11.93 -0.0832 7.1>
   rotation=<rot -0.0003 0.0002 106.7271>
  }]
 dyn_skirt_back_l_02 [{
   shapeType=Capsule
   length=41.6933
   radius=8.3545
   shapeName=dyn_skirt_back_l_02_capsule
   center=<8.6679 0.6169 9.6>
   rotation=<rot -0.0003 0.0002 106.7271>
  }]
 dyn_skirt_back_l_03 [{
   shapeType=Capsule
   length=46.9668
   radius=8.7823
   shapeName=dyn_skirt_back_l_03_capsule
   center=<3.4538 0.9573 10.7>
   rotation=<rot -0.0003 0.0002 106.7271>
  }]
 dyn_skirt_back_l_04 [{
   shapeType=Capsule
   length=57.4119
   radius=9.4035
   shapeName=dyn_skirt_back_l_04_capsule
   center=<-0.4206 0.6504 10.7>
   rotation=<rot -0.0003 0.0002 106.7271>
  }]
 dyn_skirt_back_l_MSTR [{shapeType=Capsule, length=0.505, radius=0.505, shapeName=dyn_skirt_back_l_MSTR_capsule, rotation=<rot 0 -90 0>}]
 dyn_skirt_back_mid_01 [{shapeType=Capsule, length=52.4679, radius=8.2014, shapeName=dyn_skirt_back_mid_01_capsule, center=<11.2348 0 4.6>, rotation=<rot 0 0 90>}]
 dyn_skirt_back_mid_02 [{shapeType=Capsule, length=55.0758, radius=7.9422, shapeName=dyn_skirt_back_mid_02_capsule, center=<8.079 0 4.4>, rotation=<rot 0 0 90>}]
 dyn_skirt_back_mid_03 [{shapeType=Capsule, length=57.3625, radius=8.5995, shapeName=dyn_skirt_back_mid_03_capsule, center=<1.4195 0 5.1>, rotation=<rot 0 0 90>}]
 dyn_skirt_back_mid_04 [{shapeType=Capsule, length=58.1469, radius=8.6685, shapeName=dyn_skirt_back_mid_04_capsule, center=<-1.8007 0 6.8>, rotation=<rot 0 0 90>}]
 dyn_skirt_back_mid_MSTR [{shapeType=Capsule, length=0.505, radius=0.505, shapeName=dyn_skirt_back_mid_MSTR_capsule, rotation=<rot 0 -90 0>}]
 dyn_skirt_back_r_01 [{
   shapeType=Capsule
   length=50.5238
   radius=8.464
   shapeName=dyn_skirt_back_r_01_capsule
   center=<13.3381 0.1147 7>
   rotation=<rot -0.0003 -0.0002 92.6351>
  }]
 dyn_skirt_back_r_02 [{
   shapeType=Capsule
   length=51.0065
   radius=8.473
   shapeName=dyn_skirt_back_r_02_capsule
   center=<7.9129 0.0768 7.1>
   rotation=<rot -0.0003 -0.0002 92.6351>
  }]
 dyn_skirt_back_r_03 [{
   shapeType=Capsule
   length=50.0716
   radius=9.5789
   shapeName=dyn_skirt_back_r_03_capsule
   center=<3.3509 0.0973 8.7>
   rotation=<rot -0.0003 -0.0002 92.6351>
  }]
 dyn_skirt_back_r_04 [{
   shapeType=Capsule
   length=55.0484
   radius=11.767
   shapeName=dyn_skirt_back_r_04_capsule
   center=<-0.9071 0.3952 9.3>
   rotation=<rot -0.0003 -0.0002 92.6351>
  }]
 dyn_skirt_back_r_MSTR [{shapeType=Capsule, length=0.505, radius=0.505, shapeName=dyn_skirt_back_r_MSTR_capsule, rotation=<rot 0 -90 0>}]
 dyn_skirt_front_l_01 [{
   shapeType=Capsule
   length=40.6496
   radius=7.9317
   shapeName=dyn_skirt_front_l_01_capsule
   center=<10.2293 -0.0103 7.7>
   rotation=<rot 0.0006 0.0002 74.8139>
  }]
 dyn_skirt_front_l_02 [{
   shapeType=Capsule
   length=41.933
   radius=8.1149
   shapeName=dyn_skirt_front_l_02_capsule
   center=<8.0931 -0.075 7.6>
   rotation=<rot 0.0006 0.0002 74.8139>
  }]
 dyn_skirt_front_l_03 [{
   shapeType=Capsule
   length=45.4888
   radius=8.7552
   shapeName=dyn_skirt_front_l_03_capsule
   center=<3.3024 -0.0549 7.4>
   rotation=<rot 0.0006 0.0002 74.8139>
  }]
 dyn_skirt_front_l_04 [{
   shapeType=Capsule
   length=44.6026
   radius=9.6578
   shapeName=dyn_skirt_front_l_04_capsule
   center=<-0.517 0.1573 7.4>
   rotation=<rot 0.0006 0.0002 74.8139>
  }]
 dyn_skirt_front_l_MSTR [{shapeType=Capsule, length=0.505, radius=0.505, shapeName=dyn_skirt_front_l_MSTR_capsule, rotation=<rot 0 -90 0>}]
 dyn_skirt_front_mid_01 [{shapeType=Capsule, length=35.3763, radius=8.1796, shapeName=dyn_skirt_front_mid_01_capsule, center=<12.4506 0 5.8>, rotation=<rot 0 0 90>}]
 dyn_skirt_front_mid_02 [{shapeType=Capsule, length=38.2948, radius=8.5683, shapeName=dyn_skirt_front_mid_02_capsule, center=<8.0727 0 6>, rotation=<rot 0 0 90>}]
 dyn_skirt_front_mid_03 [{shapeType=Capsule, length=40.8483, radius=8.9128, shapeName=dyn_skirt_front_mid_03_capsule, center=<3.492 0 6>, rotation=<rot 0 0 90>}]
 dyn_skirt_front_mid_04 [{shapeType=Capsule, length=40.8055, radius=10.3575, shapeName=dyn_skirt_front_mid_04_capsule, center=<-0.3497 0 6.2>, rotation=<rot 0 0 90>}]
 dyn_skirt_front_mid_MSTR [{shapeType=Capsule, length=0.505, radius=0.505, shapeName=dyn_skirt_front_mid_MSTR_capsule, rotation=<rot 0 -90 0>}]
 dyn_skirt_front_r_01 [{
   shapeType=Capsule
   length=39.4337
   radius=9.1475
   shapeName=dyn_skirt_front_r_01_capsule
   center=<10.2024 0.2146 9>
   rotation=<rot 0.0002 -0.0001 90.5582>
  }]
 dyn_skirt_front_r_02 [{
   shapeType=Capsule
   length=40.5669
   radius=9.481
   shapeName=dyn_skirt_front_r_02_capsule
   center=<8.0977 0.1575 8.9>
   rotation=<rot 0.0002 -0.0001 90.5582>
  }]
 dyn_skirt_front_r_03 [{
   shapeType=Capsule
   length=44.7903
   radius=9.4537
   shapeName=dyn_skirt_front_r_03_capsule
   center=<3.3197 -0.0348 8.2>
   rotation=<rot 0.0002 -0.0001 90.5582>
  }]
 dyn_skirt_front_r_04 [{
   shapeType=Capsule
   length=44.545
   radius=9.7155
   shapeName=dyn_skirt_front_r_04_capsule
   center=<-0.6177 -0.6537 7.2>
   rotation=<rot 0.0002 -0.0001 90.5582>
  }]
 dyn_skirt_front_r_MSTR [{shapeType=Capsule, length=0.505, radius=0.505, shapeName=dyn_skirt_front_r_MSTR_capsule, rotation=<rot 0 -90 0>}]
 dyn_skirt_ground_collision [{shapeType=Sphere, radius=5000, shapeName=root_sphere, center=<0 0 -4994>}]
 dyn_skirt_side_l_01 [{
   shapeType=Capsule
   length=51.2455
   radius=7.9273
   shapeName=dyn_skirt_side_l_01_capsule
   center=<11.387 0 8.31>
   rotation=<rot 0 0.0002 94.7942>
  }]
 dyn_skirt_side_l_02 [{
   shapeType=Capsule
   length=52.2963
   radius=7.8891
   shapeName=dyn_skirt_side_l_02_capsule
   center=<7.7015 0 8.2>
   rotation=<rot 0 0.0002 94.7943>
  }]
 dyn_skirt_side_l_03 [{
   shapeType=Capsule
   length=51.3776
   radius=8.8411
   shapeName=dyn_skirt_side_l_03_capsule
   center=<1.7972 0 8.7>
   rotation=<rot 0 0.0002 94.7943>
  }]
 dyn_skirt_side_l_04 [{
   shapeType=Capsule
   length=57.6771
   radius=9.1383
   shapeName=dyn_skirt_side_l_04_capsule
   center=<-0.9715 0 8.2>
   rotation=<rot 0 0.0002 94.7943>
  }]
 dyn_skirt_side_l_MSTR [{shapeType=Capsule, length=0.505, radius=0.505, shapeName=dyn_skirt_side_l_MSTR_capsule, rotation=<rot 0 -90 0>}]
 dyn_skirt_side_r_01 [{
   shapeType=Capsule
   length=50.098
   radius=8.0623
   shapeName=dyn_skirt_side_r_01_capsule
   center=<12.0828 0 8.3>
   rotation=<rot 0 -0.0005 78.0702>
  }]
 dyn_skirt_side_r_02 [{
   shapeType=Capsule
   length=50.8554
   radius=8.2309
   shapeName=dyn_skirt_side_r_02_capsule
   center=<8.6482 0 8.6>
   rotation=<rot 0 -0.0004 80.9473>
  }]
 dyn_skirt_side_r_03 [{
   shapeType=Capsule
   length=54.2981
   radius=8.3219
   shapeName=dyn_skirt_side_r_03_capsule
   center=<3.0801 0 8.6>
   rotation=<rot 0 -0.0001 80.3559>
  }]
 dyn_skirt_side_r_04 [{
   shapeType=Capsule
   length=56.9102
   radius=9.9052
   shapeName=dyn_skirt_side_r_04_capsule
   center=<-0.6076 -0.2471 8.2>
   rotation=<rot 0 -0.0002 79.8985>
  }]
 dyn_skirt_side_r_MSTR [{shapeType=Capsule, length=0.505, radius=0.505, shapeName=dyn_skirt_side_r_MSTR_capsule, rotation=<rot 0 -90 0>}]
 foot_l [{
   shapeType=Capsule
   length=14.4982
   radius=8.5992
   shapeName=foot_l_capsule
   center=<-3.0675 7.1637 0.1159>
   rotation=<rot -1.4848 6.1684 -87.5463>
  }]
 foot_r [{
   shapeType=Capsule
   length=12.7719
   radius=8.5992
   shapeName=foot_r_capsule
   center=<3.0835 -7.3105 -0.1096>
   rotation=<rot -1.4848 6.1684 -87.5463>
  }]
 pelvis [{
   shapeType=Capsule
   length=9.3901
   radius=11.0647
   shapeName=pelvis_capsule
   center=<-6.0152 -1.9304 -0.1097>
   rotation=<rot 1.383 -90 -0.9451>
  }]
 thigh_l [{
   shapeType=Capsule
   length=37.4932
   radius=8.0431
   shapeName=thigh_l_capsule
   center=<-20.9614 -1.1269 0.6614>
   rotation=<rot 0.041 -88.6098 -91.4166>
  }]
 thigh_r [{
   shapeType=Capsule
   length=39.5732
   radius=7.7164
   shapeName=thigh_r_capsule
   center=<23.3386 0.718 -0.9184>
   rotation=<rot 0.0236 91.1959 -87.9498>
  }]
}
constraints [
 {bone1Name=dyn_skirt_back_l_01, bone2Name=dyn_skirt_back_l_MSTR, swing1LimitDegrees=15, swing2LimitDegrees=15, twistLimitDegrees=2}
 {bone1Name=dyn_skirt_back_l_02, bone2Name=dyn_skirt_back_l_01, swing1LimitDegrees=10, swing2LimitDegrees=10, twistLimitDegrees=10}
 {bone1Name=dyn_skirt_back_l_03, bone2Name=dyn_skirt_back_l_02, swing1LimitDegrees=12, swing2LimitDegrees=12, twistLimitDegrees=10}
 {bone1Name=dyn_skirt_back_l_04, bone2Name=dyn_skirt_back_l_03, swing1LimitDegrees=12, swing2LimitDegrees=12, twistLimitDegrees=10}
 {bone1Name=dyn_skirt_back_mid_01, bone2Name=dyn_skirt_back_mid_MSTR, swing1LimitDegrees=15, swing2LimitDegrees=15, twistLimitDegrees=2}
 {bone1Name=dyn_skirt_back_mid_02, bone2Name=dyn_skirt_back_mid_01, swing1LimitDegrees=10, swing2LimitDegrees=10, twistLimitDegrees=10}
 {bone1Name=dyn_skirt_back_mid_03, bone2Name=dyn_skirt_back_mid_02, swing1LimitDegrees=12, swing2LimitDegrees=12, twistLimitDegrees=10}
 {bone1Name=dyn_skirt_back_mid_04, bone2Name=dyn_skirt_back_mid_03, swing1LimitDegrees=12, swing2LimitDegrees=12, twistLimitDegrees=10}
 {bone1Name=dyn_skirt_back_r_01, bone2Name=dyn_skirt_back_r_MSTR, swing1LimitDegrees=15, swing2LimitDegrees=15, twistLimitDegrees=2}
 {bone1Name=dyn_skirt_back_r_02, bone2Name=dyn_skirt_back_r_01, swing1LimitDegrees=10, swing2LimitDegrees=10, twistLimitDegrees=10}
 {bone1Name=dyn_skirt_back_r_03, bone2Name=dyn_skirt_back_r_02, swing1LimitDegrees=12, swing2LimitDegrees=12, twistLimitDegrees=10}
 {bone1Name=dyn_skirt_back_r_04, bone2Name=dyn_skirt_back_r_03, swing1LimitDegrees=12, swing2LimitDegrees=12, twistLimitDegrees=10}
 {
  bone1Name=dyn_skirt_front_l_01
  bone2Name=dyn_skirt_front_l_MSTR
  swing1LimitDegrees=5
  swing1Motion=Free
  swing2LimitDegrees=2.5
  swing2Motion=Free
  twistLimitDegrees=2
  twistMotion=Free
 }
 {bone1Name=dyn_skirt_front_l_02, bone2Name=dyn_skirt_front_l_01, swing1LimitDegrees=5, swing2LimitDegrees=2.5, twistLimitDegrees=10}
 {bone1Name=dyn_skirt_front_l_03, bone2Name=dyn_skirt_front_l_02, swing1LimitDegrees=5, swing2LimitDegrees=20, twistLimitDegrees=10}
 {bone1Name=dyn_skirt_front_l_04, bone2Name=dyn_skirt_front_l_03, swing1LimitDegrees=10, swing2LimitDegrees=20, twistLimitDegrees=10}
 {
  bone1Name=dyn_skirt_front_mid_01
  bone2Name=dyn_skirt_front_mid_MSTR
  swing1LimitDegrees=1
  swing1Motion=Free
  swing2LimitDegrees=2.5
  swing2Motion=Free
  twistLimitDegrees=2
  twistMotion=Free
 }
 {bone1Name=dyn_skirt_front_mid_02, bone2Name=dyn_skirt_front_mid_01, swing1LimitDegrees=1, swing2LimitDegrees=2.5, twistLimitDegrees=6}
 {bone1Name=dyn_skirt_front_mid_03, bone2Name=dyn_skirt_front_mid_02, swing1LimitDegrees=1, swing2LimitDegrees=20, twistLimitDegrees=10}
 {bone1Name=dyn_skirt_front_mid_04, bone2Name=dyn_skirt_front_mid_03, swing1LimitDegrees=10, swing2LimitDegrees=20, twistLimitDegrees=15}
 {
  bone1Name=dyn_skirt_front_r_01
  bone2Name=dyn_skirt_front_r_MSTR
  swing1LimitDegrees=5
  swing1Motion=Free
  swing2LimitDegrees=5
  swing2Motion=Free
  twistLimitDegrees=2
  twistMotion=Free
 }
 {bone1Name=dyn_skirt_front_r_02, bone2Name=dyn_skirt_front_r_01, swing1LimitDegrees=5, swing2LimitDegrees=2.5, twistLimitDegrees=10}
 {bone1Name=dyn_skirt_front_r_03, bone2Name=dyn_skirt_front_r_02, swing1LimitDegrees=5, swing2LimitDegrees=15, twistLimitDegrees=10}
 {bone1Name=dyn_skirt_front_r_04, bone2Name=dyn_skirt_front_r_03, swing1LimitDegrees=10, swing2LimitDegrees=15, twistLimitDegrees=10}
 {bone1Name=dyn_skirt_side_l_01, bone2Name=dyn_skirt_side_l_MSTR, swing1LimitDegrees=10, swing2LimitDegrees=10, twistLimitDegrees=2}
 {bone1Name=dyn_skirt_side_l_02, bone2Name=dyn_skirt_side_l_01, swing1LimitDegrees=10, swing2LimitDegrees=10, twistLimitDegrees=10}
 {bone1Name=dyn_skirt_side_l_03, bone2Name=dyn_skirt_side_l_02, swing1LimitDegrees=12, swing2LimitDegrees=12, twistLimitDegrees=10}
 {bone1Name=dyn_skirt_side_l_04, bone2Name=dyn_skirt_side_l_03, swing1LimitDegrees=12, swing2LimitDegrees=12, twistLimitDegrees=10}
 {bone1Name=dyn_skirt_side_r_01, bone2Name=dyn_skirt_side_r_MSTR, swing1LimitDegrees=10, swing2LimitDegrees=10, twistLimitDegrees=2}
 {bone1Name=dyn_skirt_side_r_02, bone2Name=dyn_skirt_side_r_01, swing1LimitDegrees=10, swing2LimitDegrees=10, twistLimitDegrees=10}
 {bone1Name=dyn_skirt_side_r_03, bone2Name=dyn_skirt_side_r_02, swing1LimitDegrees=12, swing2LimitDegrees=12, twistLimitDegrees=10}
 {bone1Name=dyn_skirt_side_r_04, bone2Name=dyn_skirt_side_r_03, swing1LimitDegrees=12, swing2LimitDegrees=12, twistLimitDegrees=10}
 {
  bone1Name=dyn_skirt_front_mid_02
  bone2Name=dyn_skirt_front_r_02
  swing1LimitDegrees=45
  swing1Motion=Free
  swing2LimitDegrees=45
  swing2Motion=Free
  twistLimitDegrees=45
  twistMotion=Free
 }
 {
  bone1Name=dyn_skirt_front_mid_02
  bone2Name=dyn_skirt_front_l_02
  swing1LimitDegrees=45
  swing1Motion=Free
  swing2LimitDegrees=45
  swing2Motion=Free
  twistLimitDegrees=45
  twistMotion=Free
 }
]
