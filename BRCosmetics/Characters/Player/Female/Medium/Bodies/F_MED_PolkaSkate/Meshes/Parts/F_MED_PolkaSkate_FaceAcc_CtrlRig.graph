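# Control Rig for the PolkaSkate female face accessory mesh that corrects braid placement as the head and body move.
# Three spherical pose readers track head orientation against spine_05, each feeding an interpolated weight into corrective transforms on bones including both braid masters.
# A weighted parent constraint makes for_aim_r follow torso and right-shoulder bones.
# When Idle Animation in Front End is true, that bone is offset and the right braid master aims at it.
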ControlRigBlueprint /BRCosmetics/Characters/Player/Female/Medium/Bodies/F_MED_PolkaSkate/Meshes/Parts/F_MED_PolkaSkate_FaceAcc_CtrlRig
graphs [
 RigVMFunctionLibrary {links [], nodes []}
 RigVMModel {nodes [
   RigUnit_BeginExecution = &RigUnit_BeginExecution()
   n1 = &RigUnit_SphericalPoseReader(
    ActiveRegionScaleFactors=(PositiveWidth=1,NegativeWidth=1,PositiveHeight=3,NegativeHeight=3)
    ActiveRegionSize=0.05
    Debug=(bDrawDebug=false,bDraw2D=False,bDrawLocalAxes=False,DebugScale=25,DebugSegments=20,DebugThickness=0.25)
    DriverAxis=(X=0,Y=1,Z=0)
    DriverItem=(Type=Bone,Name="head")
    FalloffRegionScaleFactors=(PositiveWidth=0.5,NegativeWidth=0.5,PositiveHeight=3,NegativeHeight=3)
    FalloffSize=0.2
    FlipHeightScaling=False
    FlipWidthScaling=False
    OptionalParentItem=(Type=Bone,Name="spine_05")
    RotationOffset=(X=90,Y=0,Z=15)
    ExecutePin=@RigUnit_BeginExecution.ExecutePin
   )
   RigVMFunction_AlphaInterp = &RigVMFunction_AlphaInterp(InterpSpeedDecreasing=10, InterpSpeedIncreasing=10, Scale=1, bClampResult=False, bInterpResult=True, Value=@n1.OutputParam)
   > &RigUnit_ModifyTransforms(
    ItemToModify="((Item=(Type=Bone,Name=\"dyn_braid_l_MSTR\"),Transform=(Rotation=(X=0.0076,Y=-0.1735,Z=0.043,W=0.9839),Translation=(X=0,Y=0,Z=0),Scale3D=(X=1,Y=1,Z=1))),(Item=(Type=Bone,Name=\"dyn_braid_r_MSTR\"),Transform=(Rotation=(X=0.0226,Y=-0.2578,Z=0.0842,W=0.9623),Tran ...(truncated)"
    WeightMinimum=0
    Weight=@RigVMFunction_AlphaInterp.Result
   )
   > n2 = &RigUnit_SphericalPoseReader(
    ActiveRegionScaleFactors=(PositiveWidth=1,NegativeWidth=1,PositiveHeight=1,NegativeHeight=1)
    ActiveRegionSize=0.05
    Debug=(bDrawDebug=false,bDraw2D=False,bDrawLocalAxes=False,DebugScale=25,DebugSegments=20,DebugThickness=0.25)
    DriverAxis=(X=0,Y=1,Z=0)
    DriverItem=(Type=Bone,Name="head")
    FalloffRegionScaleFactors=(PositiveWidth=0.5,NegativeWidth=0.5,PositiveHeight=1.5,NegativeHeight=1.5)
    FalloffSize=0.2
    FlipHeightScaling=False
    FlipWidthScaling=False
    OptionalParentItem=(Type=Bone,Name="spine_05")
    RotationOffset=(X=130,Y=0,Z=15)
   )
   n3 = &RigVMFunction_AlphaInterp(InterpSpeedDecreasing=10, InterpSpeedIncreasing=10, Scale=1, bClampResult=False, bInterpResult=True, Value=@n2.OutputParam)
   > &RigUnit_ModifyTransforms(
    ItemToModify="((Item=(Type=Bone,Name=\"dyn_braid_l_MSTR\"),Transform=(Rotation=(X=0.0114,Y=-0.0864,Z=0.13,W=0.9877),Translation=(X=0,Y=0,Z=0),Scale3D=(X=1,Y=1,Z=1))),(Item=(Type=Bone,Name=\"dyn_braid_r_MSTR\"),Transform=(Rotation=(X=0,Y=-0.0872,Z=0,W=0.9962),Translation=(X= ...(truncated)"
    WeightMinimum=0
    Weight=@n3.Result
   )
   > n4 = &RigUnit_SphericalPoseReader(
    ActiveRegionScaleFactors=(PositiveWidth=1,NegativeWidth=1,PositiveHeight=1,NegativeHeight=1)
    ActiveRegionSize=0.05
    Debug=(bDrawDebug=false,bDraw2D=False,bDrawLocalAxes=False,DebugScale=25,DebugSegments=20,DebugThickness=0.25)
    DriverAxis=(X=0,Y=1,Z=0)
    DriverItem=(Type=Bone,Name="head")
    FalloffRegionScaleFactors=(PositiveWidth=0.5,NegativeWidth=0.5,PositiveHeight=1.5,NegativeHeight=1.5)
    FalloffSize=0.2
    FlipHeightScaling=False
    FlipWidthScaling=False
    OptionalParentItem=(Type=Bone,Name="spine_05")
    RotationOffset=(X=50,Y=0,Z=15)
   )
   n5 = &RigVMFunction_AlphaInterp(InterpSpeedDecreasing=10, InterpSpeedIncreasing=10, Scale=1, bClampResult=False, bInterpResult=True, Value=@n4.OutputParam)
   > &RigUnit_ModifyTransforms(
    ItemToModify="((Item=(Type=Bone,Name=\"dyn_braid_l_MSTR\"),Transform=(Rotation=(X=0.0114,Y=-0.0864,Z=0.13,W=0.9877),Translation=(X=0,Y=0,Z=0),Scale3D=(X=1,Y=1,Z=1))),(Item=(Type=Bone,Name=\"dyn_braid_r_MSTR\"),Transform=(Rotation=(X=0,Y=-0.1736,Z=0,W=0.9848),Translation=(X= ...(truncated)"
    WeightMinimum=0
    Weight=@n5.Result
   )
   > &RigUnit_ParentConstraint(
    AdvancedSettings=(RotationOrderForFilter=XZY)
    Child=(Type=Bone,Name="for_aim_r")
    Parents="((Item=(Type=Bone,Name=\"spine_03\"),Weight=0.0156),(Item=(Type=Bone,Name=\"spine_02\"),Weight=0.0188),(Item=(Type=Bone,Name=\"deform_upperarm_fr_r\"),Weight=0.0241),(Item=(Type=Bone,Name=\"upperarm_twist_01_r\"),Weight=0.0456),(Item=(Type=Bone,Name=\"clavicle_r\"), ...(truncated)"
    Weight=1
   )
   RigVMFunction_MathBoolToFloat = &RigVMFunction_MathBoolToFloat(Value=RigVMVariableNode() {variable=Idle Animation in Front End}.Value)
   > &RigUnit_ModifyTransforms(
    ItemToModify=((Item=(Type=Bone,Name="for_aim_r"),Transform=(Rotation=(X=0,Y=0,Z=0,W=-1),Translation=(X=0,Y=-3.0459,Z=1.5531),Scale3D=(X=1,Y=1,Z=1))))
    WeightMinimum=0
    Weight=@RigVMFunction_MathBoolToFloat.Result
   )
   > &RigUnit_AimConstraintLocalSpaceOffset(
    Child=(Type=Bone,Name="dyn_braid_r_MSTR")
    Filter=(bX=True,bY=True,bZ=True)
    Parents=((Item=(Type=Bone,Name="for_aim_r"),Weight=1))
    WorldUp=(Target=(X=0,Y=0,Z=1),Kind=Direction,Space=(Type=Bone,Name="for_aim_r"))
    Weight=@RigVMFunction_MathBoolToFloat.Result
   )
  ]}
]
hierarchy [
 root {loc=<0 0 0>, rot=<0 0 0 1>}
 pelvis {parent=root, loc=<0 -0.5301 98.0571>, rot=<0.0127 0.707 -0.0127 -0.707>}
 spine_01 {parent=pelvis, loc=<6.1208 -0.9328 0>, rot=<0 0 0.034 -0.9994>}
 spine_02 {parent=spine_01, loc=<6.5997 -0.2095 0>, rot=<0 0 0.0238 -0.9997>}
 spine_03 {parent=spine_02, loc=<6.4914 -0.1237 0>, rot=<0 0 0.0149 -0.9999>}
 spine_04 {parent=spine_03, loc=<6.9598 0.0794 0>, rot=<0 0 -0.0132 -0.9999>}
 spine_05 {parent=spine_04, loc=<6.9987 0 0>, rot=<0 0 -0.0015 -1>}
 neck_01 {parent=spine_05, loc=<11.3138 0.1529 0>, rot=<0 0 -0.2003 -0.9797>}
 neck_02 {parent=neck_01, loc=<4.5627 0 0>, rot=<0 0 0.0504 -0.9987>}
 head {parent=neck_02, loc=<5.3914 0 0>, rot=<0 0 0.115 -0.9934>}
 dyn_braid_r_MSTR {parent=head, loc=<-0.5581 -2.0793 6.6079>, rot=<0.1315 0.6469 0.7503 0.0347>}
 dyn_braid_r_01 {parent=dyn_braid_r_MSTR, loc=<0 0 0>, rot=<0 0 0 1>}
 dyn_braid_r_02 {parent=dyn_braid_r_01, loc=<6.452 0 0>, rot=<-0.0142 0.1129 0.0248 -0.9932>}
 dyn_braid_r_03 {parent=dyn_braid_r_02, loc=<5.697 0 0>, rot=<0.0593 0.0801 0.0105 -0.995>}
 dyn_braid_r_04 {parent=dyn_braid_r_03, loc=<6.0248 0 0>, rot=<0.0209 -0.0321 0.0145 -0.9992>}
 dyn_braid_r_05 {parent=dyn_braid_r_04, loc=<7.0342 0 0>, rot=<0.0436 0 0 -0.999>}
 dyn_braid_l_MSTR {parent=head, loc=<-0.5583 -2.0793 -6.6079>, rot=<-0.0308 -0.7199 -0.6792 -0.1398>}
 dyn_braid_l_01 {parent=dyn_braid_l_MSTR, loc=<0 0 0>, rot=<0 0 0 1>}
 dyn_braid_l_02 {parent=dyn_braid_l_01, loc=<6.4328 0 0>, rot=<-0.0205 0.1088 -0.0042 -0.9938>}
 dyn_braid_l_03 {parent=dyn_braid_l_02, loc=<5.6968 0 0>, rot=<-0.0227 0.0498 -0.0302 -0.998>}
 dyn_braid_l_04 {parent=dyn_braid_l_03, loc=<6.0288 0 0>, rot=<-0.0095 0.0133 -0.0225 -0.9996>}
 dyn_braid_l_05 {parent=dyn_braid_l_04, loc=<7.0284 0.0027 0.036>, rot=<-0.1342 -0.03 0.0064 -0.9905>}
 dyn_bangs_l_MSTR {parent=head, loc=<9.5915 8.9254 -2.8909>, rot=<0.045 -0.8621 -0.4644 -0.1977>}
 dyn_bangs_l {parent=dyn_bangs_l_MSTR, loc=<0 0 0>, rot=<0 0 0 1>}
 dyn_bangs_l_in {parent=dyn_bangs_l, loc=<2.3388 1.4734 -0.1665>, rot=<0.1178 -0.1212 0.0676 -0.9833>}
 dyn_bangs_l_out {parent=dyn_bangs_l, loc=<2.6026 -1.5037 -0.3695>, rot=<-0.0782 -0.156 0.0478 -0.9835>}
 dyn_bangs_r_MSTR {parent=head, loc=<9.6021 9.3122 2.6068>, rot=<0.1399 0.551 0.8164 -0.1018>}
 dyn_bangs_r {parent=dyn_bangs_r_MSTR, loc=<0 0 0>, rot=<0 0 0 1>}
 dyn_bangs_r_out {parent=dyn_bangs_r, loc=<2.6886 1.0001 -0.3107>, rot=<0.1406 -0.1924 -0.0354 -0.9705>}
 dyn_bangs_r_in {parent=dyn_bangs_r, loc=<2.745 -1.7327 -0.0415>, rot=<-0.1101 -0.1561 -0.0491 -0.9803>}
 dyn_hair_strand_l_small_MSTR {parent=head, loc=<5.6511 6.4654 -6.7957>, rot=<0.0447 -0.977 -0.205 0.0384>}
 dyn_hair_strand_l_small_01 {parent=dyn_hair_strand_l_small_MSTR, loc=<0 0 0>, rot=<0 0 0 1>}
 dyn_hair_strand_l_small_02 {parent=dyn_hair_strand_l_small_01, loc=<2.2181 0 0>, rot=<-0.0354 -0.0445 0.1023 -0.9931>}
 dyn_hair_strand_l_large_MSTR {parent=head, loc=<5.5078 4.6182 -7.1691>, rot=<0.2401 -0.9617 -0.124 0.0449>}
 dyn_hair_strand_l_large_01 {parent=dyn_hair_strand_l_large_MSTR, loc=<0 0 0>, rot=<0 0 0 1>}
 dyn_hair_strand_l_large_02 {parent=dyn_hair_strand_l_large_01, loc=<3.4453 0 0>, rot=<0.0549 -0.0753 -0.1565 -0.9833>}
 dyn_hair_strand_l_large_03 {parent=dyn_hair_strand_l_large_02, loc=<3.183 0 0>, rot=<-0.1175 0.0997 -0.2075 -0.966>}
 dyn_hair_strand_r_MSTR {parent=head, loc=<4.4273 4.4923 6.9748>, rot=<-0.081 0.1571 0.9601 -0.2167>}
 dyn_hair_strand_r_01 {parent=dyn_hair_strand_r_MSTR, loc=<0 0 0>, rot=<0 0 0 1>}
 dyn_hair_strand_r_02 {parent=dyn_hair_strand_r_01, loc=<3.4028 0 0>, rot=<-0.0585 0.0044 0.0877 -0.9944>}
 dyn_hair_strand_r_03 {parent=dyn_hair_strand_r_02, loc=<3.1883 0 0>, rot=<0.0479 0.2968 0.152 -0.9416>}
 clavicle_l {parent=spine_05, loc=<5.4831 -0.1053 -3.2946>, rot=<0.0465 -0.7289 -0.009 -0.6829>}
 upperarm_l {parent=clavicle_l, loc=<9.8658 -1.2136 -0.4105>, rot=<0.0164 -0.3245 -0.0269 -0.9454>}
 lowerarm_l {parent=upperarm_l, loc=<26.465 -0.8345 0.259>, rot=<0.0036 0.0161 -0.2298 -0.9731>}
 upperarm_twist_01_l {parent=upperarm_l, loc=<7.2135 -0.1134 0>, rot=<0 0 0 1>}
 deform_upperarm_fr_base_l {parent=upperarm_twist_01_l, loc=<-5.6312 5.092 0.3354>, rot=<0 0 0 -1>}
 deform_upperarm_fr_l {parent=deform_upperarm_fr_base_l, loc=<0 0 0>, rot=<0 0 0 1>}
 deform_upperarm_up_base_l {parent=upperarm_twist_01_l, loc=<-8.891 0.4825 3.9517>, rot=<0 0 0 -1>}
 deform_upperarm_up_l {parent=deform_upperarm_up_base_l, loc=<0 0 0>, rot=<0 0 0 1>}
 clavicle_r {parent=spine_05, loc=<5.4831 -0.1053 3.2946>, rot=<0.7289 0.0465 -0.6829 0.009>}
 upperarm_r {parent=clavicle_r, loc=<-9.8658 1.2136 0.4105>, rot=<0.0164 -0.3245 -0.0269 -0.9454>}
 lowerarm_r {parent=upperarm_r, loc=<-26.465 0.8345 -0.259>, rot=<0.0036 0.0161 -0.2298 -0.9731>}
 upperarm_twist_01_r {parent=upperarm_r, loc=<-7.2135 0.1134 0>, rot=<0 0 0 -1>}
 deform_upperarm_fr_base_r {parent=upperarm_twist_01_r, loc=<5.6312 -5.092 -0.3354>, rot=<0 0 0 -1>}
 deform_upperarm_fr_r {parent=deform_upperarm_fr_base_r, loc=<0 0 0>, rot=<0 0 0 1>}
 deform_upperarm_up_base_r {parent=upperarm_twist_01_r, loc=<8.891 -0.4825 -3.9517>, rot=<0 0 0 -1>}
 deform_upperarm_up_r {parent=deform_upperarm_up_base_r, loc=<0 0 0>, rot=<0 0 0 1>}
 deform_thing_A_wrist {parent=clavicle_r, loc=<-14.1532 3.5716 -15.59>, rot=<0.6664 -0.6208 0.1611 0.3801>}
 deform_thing_A_hand {parent=deform_thing_A_wrist, loc=<0 0 0>, rot=<0 0 0 1>}
 deform_thing_A_thumb_01 {parent=deform_thing_A_hand, loc=<2.6885 -3.5662 -1.5723>, rot=<-0.1157 -0.3522 0.2349 -0.8986>}
 deform_thing_A_thumb_02 {parent=deform_thing_A_thumb_01, loc=<3.4825 0 0>, rot=<0 0 -0.206 -0.9786>}
 deform_thing_A_thumb_03 {parent=deform_thing_A_thumb_02, loc=<2.8996 0 0>, rot=<0 0 -0.0668 -0.9978>}
 deform_thing_A_index_metacarpal {parent=deform_thing_A_hand, loc=<1.9918 -3.3183 -0.6339>, rot=<0.6552 0.0388 0.0117 -0.7543>}
 deform_thing_A_index_01 {parent=deform_thing_A_index_metacarpal, loc=<7.4719 0 0>, rot=<0 0 -0.1468 -0.9892>}
 deform_thing_A_index_02 {parent=deform_thing_A_index_01, loc=<4.1266 0 0>, rot=<0 0 -0.3324 -0.9431>}
 deform_thing_A_index_03 {parent=deform_thing_A_index_02, loc=<2.6864 0 0>, rot=<0 0 -0.2036 -0.9791>}
 deform_thing_A_middle_metacarpal {parent=deform_thing_A_hand, loc=<2.2933 -1.6479 0.2632>, rot=<0.6843 0.0527 -0.0463 -0.7258>}
 deform_thing_A_middle_01 {parent=deform_thing_A_middle_metacarpal, loc=<7.0419 0 0>, rot=<0 0 -0.1092 -0.994>}
 deform_thing_A_middle_02 {parent=deform_thing_A_middle_01, loc=<4.5892 0 0>, rot=<0 0 -0.2986 -0.9544>}
 deform_thing_A_middle_03 {parent=deform_thing_A_middle_02, loc=<2.6375 0 0>, rot=<0 0 -0.2234 -0.9747>}
 deform_thing_A_ring_metacarpal {parent=deform_thing_A_hand, loc=<2.2496 0.9403 0.4665>, rot=<0.7384 0.0457 -0.0664 -0.6695>}
 deform_thing_A_ring_01 {parent=deform_thing_A_ring_metacarpal, loc=<6.7033 0 0>, rot=<0 0 -0.1416 -0.9899>}
 deform_thing_A_ring_02 {parent=deform_thing_A_ring_01, loc=<4.1845 0 0>, rot=<0 0 -0.2175 -0.9761>}
 deform_thing_A_ring_03 {parent=deform_thing_A_ring_02, loc=<2.3557 0 0>, rot=<0 0 -0.1671 -0.9859>}
 deform_thing_A_pinky_metacarpal {parent=deform_thing_A_hand, loc=<1.9337 2.6862 -0.1754>, rot=<0.7676 0.0649 -0.0896 -0.6313>}
 deform_thing_A_pinky_01 {parent=deform_thing_A_pinky_metacarpal, loc=<6.6008 0 0>, rot=<0 0 -0.0962 -0.9954>}
 deform_thing_A_pinky_02 {parent=deform_thing_A_pinky_01, loc=<3.5766 0 0>, rot=<0 0 -0.3162 -0.9487>}
 deform_thing_A_pinky_03 {parent=deform_thing_A_pinky_02, loc=<1.9766 0 0>, rot=<0 0 0 -1>}
 deform_pec_aim_base_l {parent=spine_04, loc=<3.2456 9.3309 -1.2237>, rot=<0.0603 -0.4033 0.3318 -0.8507>}
 deform_pec_aim_l {parent=deform_pec_aim_base_l, loc=<0 0 0>, rot=<0 0 0 1>}
 deform_pec_base_l {parent=deform_pec_aim_l, loc=<9.3665 0 0>, rot=<0 0 0 1>}
 deform_pec_l {parent=deform_pec_base_l, loc=<0 0 0>, rot=<0 0 0 1>}
 deform_pec_aim_base_r {parent=spine_04, loc=<3.2452 9.3309 1.2237>, rot=<-0.4033 -0.0603 0.8507 0.3318>}
 deform_pec_aim_r {parent=deform_pec_aim_base_r, loc=<0 0 0>, rot=<0 0 0 1>}
 deform_pec_base_r {parent=deform_pec_aim_r, loc=<-9.3667 -0.0003 -0.0004>, rot=<0 0 0 1>}
 deform_pec_r {parent=deform_pec_base_r, loc=<0 0 0>, rot=<0 0 0 1>}
 for_aim_r {parent=root, loc=<-9.6363 7.5924 129.491>, rot=<0.41 -0.2507 -0.6919 -0.5388>}
 thigh_l_up_SPR {type=Curve, loc=<0 0 0>, rot=<0 0 0 1>}
 thigh_r_up_SPR {type=Curve, loc=<0 0 0>, rot=<0 0 0 1>}
]
variables [Idle Animation in Front End {type=bool, default=False}]
