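# Physics asset for the front-end variant of the PolkaSkate female character's face accessory, simulating two dangling hair braids.
# Each braid is a five-segment capsule chain from a master bone, with swing and twist limits mostly tighter on the left.
# Capsules on the upper spine, neck, clavicles, upper arms and one deform_thing_A hand with fingers supply body shapes around the braids.
# Extra free-angular constraints tie braid segments to spine_05 and an index finger bone.
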PhysicsAsset /BRCosmetics/Characters/Player/Female/Medium/Bodies/F_MED_PolkaSkate/Meshes/Parts/F_MED_PolkaSkate_FaceAcc_FrontEnd_Physics
bodies {
 clavicle_l [{
   shapeType=Capsule
   length=14.2097
   radius=6.567
   shapeName=clavicle_l_capsule
   center=<4.9167 -3.0019 -0.4457>
   rotation=<rot 66.8904 -0.0016 168.6191>
  }]
 clavicle_r [{
   shapeType=Capsule
   length=14.2097
   radius=6.567
   shapeName=clavicle_r_capsule
   center=<-4.9167 3.0019 0.4457>
   rotation=<rot 66.8904 -0.0016 -11.3809>
  }]
 deform_thing_A_hand [{
   shapeType=Capsule
   length=4.2114
   radius=2.7891
   shapeName=deform_thing_A_hand_capsule
   center=<8.4564 0.7185 -0.4998>
   rotation=<rot 0.0005 -0.0002 -83.7511>
  }]
 deform_thing_A_index_02 [{shapeType=Capsule, length=5.1974, radius=1.2404, shapeName=deform_thing_A_index_02_capsule, rotation=<rot 90 -23.0368 0>}]
 deform_thing_A_index_03 [{shapeType=Capsule, length=7.5386, radius=0.7784, shapeName=deform_thing_A_index_03_capsule, center=<1.3815 0 0>, rotation=<rot 90 0 0>}]
 deform_thing_A_middle_02 [{shapeType=Capsule, length=4.0562, radius=2.3815, shapeName=deform_thing_A_middle_02_capsule, rotation=<rot 90 -23.0368 0>}]
 deform_thing_A_middle_03 [{
   shapeType=Capsule
   length=8.032
   radius=2.4368
   shapeName=deform_thing_A_middle_03_capsule
   center=<1.852 0 -0.1203>
   rotation=<rot 86.2842 -0.0002 -0.0002>
  }]
 deform_thing_A_pinky_02 [{shapeType=Capsule, length=5.3403, radius=1.0974, shapeName=deform_thing_A_pinky_02_capsule, rotation=<rot 90 -23.0368 0>}]
 deform_thing_A_pinky_03 [{shapeType=Capsule, length=3.3367, radius=1.0974, shapeName=deform_thing_A_pinky_03_capsule, rotation=<rot 90 0 0>}]
 deform_thing_A_ring_02 [{shapeType=Capsule, length=5.3403, radius=1.0974, shapeName=deform_thing_A_ring_02_capsule, rotation=<rot 90 -23.0368 0>}]
 deform_thing_A_ring_03 [{shapeType=Capsule, length=9.9043, radius=0.9602, shapeName=deform_thing_A_ring_03_capsule, center=<1.8559 0 0>, rotation=<rot 90 0 0>}]
 deform_thing_A_thumb_02 [{shapeType=Capsule, length=2, radius=1, shapeName=deform_thing_A_thumb_02_capsule, rotation=<rot 90 0 0>}]
 deform_thing_A_thumb_03 [{shapeType=Capsule, length=2, radius=1, shapeName=deform_thing_A_thumb_03_capsule, rotation=<rot 90 0 0>}]
 dyn_braid_l_01 [{
   shapeType=Capsule
   length=6.1321
   radius=1.7981
   shapeName=dyn_braid_l_01_capsule
   center=<3.9961 0.1777 0.3796>
   rotation=<rot 0.4451 -89.6852 -85.5094>
  }]
 dyn_braid_l_02 [{
   shapeType=Capsule
   length=6.4611
   radius=1.7878
   shapeName=dyn_braid_l_02_capsule
   center=<2.7578 0.1541 0.3585>
   rotation=<rot 0 90.7121 -88.1461>
  }]
 dyn_braid_l_03 [{
   shapeType=Capsule
   length=5.0405
   radius=1.5523
   shapeName=dyn_braid_l_03_capsule
   center=<2.9652 -0.0094 0.2017>
   rotation=<rot 0 95.1634 -87.6815>
  }]
 dyn_braid_l_04 [{
   shapeType=Capsule
   length=7.0434
   radius=1.3988
   shapeName=dyn_braid_l_04_capsule
   center=<3.647 0.1287 0.387>
   rotation=<rot 0 91.5007 -86.4171>
  }]
 dyn_braid_l_05 [{
   shapeType=Capsule
   length=5.3084
   radius=1.5521
   shapeName=dyn_braid_l_05_capsule
   center=<2.6371 -0.0991 0.0127>
   rotation=<rot 0 -88.4255 -80.9336>
  }]
 dyn_braid_l_MSTR [{shapeType=Capsule, length=0.505, radius=0.505, shapeName=dyn_braid_l_MSTR_capsule, rotation=<rot 0 -90 0>}]
 dyn_braid_r_01 [{
   shapeType=Capsule
   length=2.202
   radius=1.4882
   shapeName=dyn_braid_r_01_capsule
   center=<3.6396 -0.1182 0.2867>
   rotation=<rot 0 -88.7281 -86.7204>
  }]
 dyn_braid_r_02 [{
   shapeType=Capsule
   length=2.7668
   radius=1.5645
   shapeName=dyn_braid_r_02_capsule
   center=<2.9094 0.0059 0.083>
   rotation=<rot -0.0372 90.4197 -88.8393>
  }]
 dyn_braid_r_03 [{
   shapeType=Capsule
   length=2.7776
   radius=1.3662
   shapeName=dyn_braid_r_03_capsule
   center=<3.0759 0.0454 0.0314>
   rotation=<rot 0 88.044 -87.4412>
  }]
 dyn_braid_r_04 [{
   shapeType=Capsule
   length=3.5739
   radius=1.6638
   shapeName=dyn_braid_r_04_capsule
   center=<3.439 -0.1593 0.5419>
   rotation=<rot -0.001 87.7392 -86.6869>
  }]
 dyn_braid_r_05 [{
   shapeType=Capsule
   length=3.5674
   radius=2.1343
   shapeName=dyn_braid_r_05_capsule
   center=<2.7257 0.0141 -0.3604>
   rotation=<rot 0 -87.0239 -78.2148>
  }]
 dyn_braid_r_MSTR [{shapeType=Capsule, length=0.505, radius=0.505, shapeName=dyn_braid_r_MSTR_capsule, rotation=<rot 0 -90 0>}]
 neck_01 [{shapeType=Capsule, length=6.9045, radius=4.6261, shapeName=neck_01_capsule, center=<2.0703 0 0>, rotation=<rot 90 0 0>}]
 neck_02 [{shapeType=Capsule, length=7.8058, radius=4.5997, shapeName=neck_02_capsule, center=<1.8089 0 0>, rotation=<rot 90 0 0>}]
 spine_04 [{shapeType=Capsule, length=12.3325, radius=10.3889, shapeName=spine_04_capsule, center=<-2.6125 0.4956 0>, rotation=<rot 0 -90 0>}]
 spine_05 [{shapeType=Capsule, length=8.2616, radius=10.07, shapeName=spine_05_capsule, center=<0 -1.4457 0>}]
 upperarm_l [{
   shapeType=Capsule
   length=24.9249
   radius=5.0751
   shapeName=upperarm_l_capsule
   center=<15 0 -0.7281>
   rotation=<rot 86.2733 -179.9994 0.0006>
  }]
 upperarm_r [{
   shapeType=Capsule
   length=24.9249
   radius=5.0751
   shapeName=upperarm_r_capsule
   center=<-15 0 0.7281>
   rotation=<rot 86.2733 -179.9994 -179.9994>
  }]
}
constraints [
 {bone1Name=dyn_braid_l_01, bone2Name=dyn_braid_l_MSTR, swing1LimitDegrees=0, swing2LimitDegrees=15, twistLimitDegrees=4}
 {bone1Name=dyn_braid_l_02, bone2Name=dyn_braid_l_01, swing1LimitDegrees=5, swing2LimitDegrees=15, twistLimitDegrees=4}
 {bone1Name=dyn_braid_l_03, bone2Name=dyn_braid_l_02, swing1LimitDegrees=7, swing2LimitDegrees=20, twistLimitDegrees=4}
 {bone1Name=dyn_braid_l_04, bone2Name=dyn_braid_l_03, swing1LimitDegrees=10, swing2LimitDegrees=20, twistLimitDegrees=4}
 {bone1Name=dyn_braid_l_05, bone2Name=dyn_braid_l_04, swing1LimitDegrees=15, swing2LimitDegrees=20, twistLimitDegrees=4}
 {bone1Name=dyn_braid_r_01, bone2Name=dyn_braid_r_MSTR, swing1LimitDegrees=25, swing2LimitDegrees=25, twistLimitDegrees=8}
 {bone1Name=dyn_braid_r_02, bone2Name=dyn_braid_r_01, swing1LimitDegrees=25, swing2LimitDegrees=25, twistLimitDegrees=8}
 {bone1Name=dyn_braid_r_03, bone2Name=dyn_braid_r_02, swing1LimitDegrees=25, swing2LimitDegrees=25, twistLimitDegrees=8}
 {bone1Name=dyn_braid_r_04, bone2Name=dyn_braid_r_03, swing1LimitDegrees=25, swing2LimitDegrees=25, twistLimitDegrees=8}
 {bone1Name=dyn_braid_r_05, bone2Name=dyn_braid_r_04, swing1LimitDegrees=25, swing2LimitDegrees=25, twistLimitDegrees=8}
 {
  bone1Name=dyn_braid_r_03
  bone2Name=deform_thing_A_index_03
  swing1LimitDegrees=15
  swing1Motion=Free
  swing2LimitDegrees=15
  swing2Motion=Free
  twistLimitDegrees=5
  twistMotion=Free
 }
 {
  bone1Name=dyn_braid_l_03
  bone2Name=spine_05
  swing1LimitDegrees=45
  swing1Motion=Free
  swing2LimitDegrees=45
  swing2Motion=Free
  twistLimitDegrees=45
  twistMotion=Free
 }
 {
  bone1Name=dyn_braid_r_05
  bone2Name=spine_05
  swing1LimitDegrees=45
  swing1Motion=Free
  swing2LimitDegrees=45
  swing2Motion=Free
  twistLimitDegrees=45
  twistMotion=Free
 }
]
