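# Physics asset for the front-end, without-hand face accessory mesh of the PolkaSkate female character, set up for two braids.
# Each braid is a five-bone capsule chain from its MSTR bone, with tight twist and swing limits that generally widen toward the tip.
# An extra constraint links each braid's middle bone to spine_05 with free swing and twist.
# Capsules on the head, neck, upper spine, clavicles and upper arms approximate the surrounding body.
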PhysicsAsset /BRCosmetics/Characters/Player/Female/Medium/Bodies/F_MED_PolkaSkate/Meshes/Parts/F_MED_PolkaSkate_FaceAcc_FrontEnd_WithoutHand_Physics
bodies {
 clavicle_l [{
   shapeType=Capsule
   length=17.7986
   radius=7.2142
   shapeName=clavicle_l_capsule
   center=<4.4395 -2.7459 -0.6047>
   rotation=<rot 72.9107 -4.3073 -18.0368>
  }]
 clavicle_r [{
   shapeType=Capsule
   length=14.2097
   radius=6.567
   shapeName=clavicle_r_capsule
   center=<-4.9167 3.0019 0.4457>
   rotation=<rot 66.8904 -0.0016 -11.3809>
  }]
 dyn_braid_l_01 [{
   shapeType=Capsule
   length=6.1321
   radius=1.7981
   shapeName=dyn_braid_l_01_capsule
   center=<3.9961 0.1777 0.3796>
   rotation=<rot 0.4451 -89.6852 -85.5094>
  }]
 dyn_braid_l_02 [{
   shapeType=Capsule
   length=6.4611
   radius=1.7878
   shapeName=dyn_braid_l_02_capsule
   center=<2.7578 0.1541 0.3585>
   rotation=<rot 0 90.7121 -88.1461>
  }]
 dyn_braid_l_03 [{
   shapeType=Capsule
   length=5.0405
   radius=1.5523
   shapeName=dyn_braid_l_03_capsule
   center=<2.9652 -0.0094 0.2017>
   rotation=<rot 0 95.1634 -87.6815>
  }]
 dyn_braid_l_04 [{
   shapeType=Capsule
   length=7.2035
   radius=1.2386
   shapeName=dyn_braid_l_04_capsule
   center=<3.647 0.1287 0.387>
   rotation=<rot 0 91.5007 -86.4171>
  }]
 dyn_braid_l_05 [{
   shapeType=Capsule
   length=5.3084
   radius=1.5521
   shapeName=dyn_braid_l_05_capsule
   center=<2.6371 -0.0991 0.0127>
   rotation=<rot 0 -88.4255 -80.9336>
  }]
 dyn_braid_l_MSTR [{shapeType=Capsule, length=0.505, radius=0.505, shapeName=dyn_braid_l_MSTR_capsule, rotation=<rot 0 -90 0>}]
 dyn_braid_r_01 [{
   shapeType=Capsule
   length=3.7278
   radius=1.164
   shapeName=dyn_braid_r_01_capsule
   center=<3.6396 -0.1182 0.2867>
   rotation=<rot 0 -88.7281 -86.7204>
  }]
 dyn_braid_r_02 [{
   shapeType=Capsule
   length=3.6639
   radius=1.1974
   shapeName=dyn_braid_r_02_capsule
   center=<2.9094 0.0059 0.083>
   rotation=<rot -0.0372 90.4197 -88.8393>
  }]
 dyn_braid_r_03 [{
   shapeType=Capsule
   length=3.8707
   radius=0.9207
   shapeName=dyn_braid_r_03_capsule
   center=<3.0759 0.0454 0.0314>
   rotation=<rot 0 88.044 -87.4412>
  }]
 dyn_braid_r_04 [{
   shapeType=Capsule
   length=6.5634
   radius=1.3936
   shapeName=dyn_braid_r_04_capsule
   center=<3.439 -0.1593 0.5419>
   rotation=<rot -0.001 87.7392 -86.6869>
  }]
 dyn_braid_r_05 [{
   shapeType=Capsule
   length=4.6607
   radius=1.7821
   shapeName=dyn_braid_r_05_capsule
   center=<2.7257 0.0141 -0.3604>
   rotation=<rot 0 -87.0239 -78.2148>
  }]
 dyn_braid_r_MSTR [{shapeType=Capsule, length=0.505, radius=0.505, shapeName=dyn_braid_r_MSTR_capsule, rotation=<rot 0 -90 0>}]
 head [{shapeType=Capsule, length=10.6523, radius=4.8879, shapeName=head_capsule, center=<3.0246 0.1626 0.1116>, rotation=<rot 90 0 0>}]
 neck_01 [{shapeType=Capsule, length=6.7214, radius=4.8093, shapeName=neck_01_capsule, center=<2.0703 0 0>, rotation=<rot 90 0 0>}]
 neck_02 [{shapeType=Capsule, length=7.5758, radius=4.8297, shapeName=neck_02_capsule, center=<1.8089 0 0>, rotation=<rot 90 0 0>}]
 spine_04 [{shapeType=Capsule, length=12.3325, radius=10.3889, shapeName=spine_04_capsule, center=<-2.6125 0.4956 0>, rotation=<rot 0 -90 0>}]
 spine_05 [{shapeType=Capsule, length=8.1369, radius=10.3316, shapeName=spine_05_capsule, center=<-2.898 -0.209 0>, rotation=<rot 0 -90 0>}]
 upperarm_l [{
   shapeType=Capsule
   length=24.8452
   radius=5.1548
   shapeName=upperarm_l_capsule
   center=<15 0 -0.6418>
   rotation=<rot 87.1276 -179.9987 -179.9988>
  }]
 upperarm_r [{
   shapeType=Capsule
   length=24.9249
   radius=5.0751
   shapeName=upperarm_r_capsule
   center=<-15 0 0.7281>
   rotation=<rot 86.2733 -179.9994 -179.9994>
  }]
}
constraints [
 {bone1Name=dyn_braid_l_01, bone2Name=dyn_braid_l_MSTR, swing1LimitDegrees=0, swing2LimitDegrees=10, twistLimitDegrees=4}
 {bone1Name=dyn_braid_l_02, bone2Name=dyn_braid_l_01, swing1LimitDegrees=5, swing2LimitDegrees=10, twistLimitDegrees=4}
 {bone1Name=dyn_braid_l_03, bone2Name=dyn_braid_l_02, swing1LimitDegrees=7, swing2LimitDegrees=15, twistLimitDegrees=4}
 {bone1Name=dyn_braid_l_04, bone2Name=dyn_braid_l_03, swing1LimitDegrees=10, swing2LimitDegrees=20, twistLimitDegrees=4}
 {bone1Name=dyn_braid_l_05, bone2Name=dyn_braid_l_04, swing1LimitDegrees=15, swing2LimitDegrees=20, twistLimitDegrees=4}
 {bone1Name=dyn_braid_r_01, bone2Name=dyn_braid_r_MSTR, swing1LimitDegrees=0, swing2LimitDegrees=10, twistLimitDegrees=4}
 {bone1Name=dyn_braid_r_02, bone2Name=dyn_braid_r_01, swing1LimitDegrees=5, swing2LimitDegrees=10, twistLimitDegrees=4}
 {bone1Name=dyn_braid_r_03, bone2Name=dyn_braid_r_02, swing1LimitDegrees=7, swing2LimitDegrees=15, twistLimitDegrees=4}
 {bone1Name=dyn_braid_r_04, bone2Name=dyn_braid_r_03, swing1LimitDegrees=5, swing2LimitDegrees=15, twistLimitDegrees=4}
 {bone1Name=dyn_braid_r_05, bone2Name=dyn_braid_r_04, swing1LimitDegrees=8, swing2LimitDegrees=20, twistLimitDegrees=5}
 {
  bone1Name=dyn_braid_l_03
  bone2Name=spine_05
  swing1LimitDegrees=45
  swing1Motion=Free
  swing2LimitDegrees=45
  swing2Motion=Free
  twistLimitDegrees=45
  twistMotion=Free
 }
 {
  bone1Name=dyn_braid_r_03
  bone2Name=spine_05
  swing1LimitDegrees=45
  swing1Motion=Free
  swing2LimitDegrees=45
  swing2Motion=Free
  twistLimitDegrees=45
  twistMotion=Free
 }
]
